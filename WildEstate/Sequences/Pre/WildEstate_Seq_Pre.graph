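# Master Level Sequence for the WildEstate "Pre" cinematic, running at 60 fps with an eleven-section cinematic shot track that includes gaps and overlaps.
# The WildEstate_House_H building level instance is hidden during the first shot, while the Sq_WildEstate_House_H sub-sequence plays, then shown again.
# A Data Layer track has two back-to-back sections, one for that first shot and one spanning the following five shots.
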LevelSequence /WildEstate/Sequences/Pre/WildEstate_Seq_Pre
bindings [WildEstate_House_H {
  id=#1
  possessed_class=BuildingLevelInstance
  tracks [{
    type=MovieSceneVisibilityTrack
    sections [{type=MovieSceneVisibilitySection, channels [None {type=MovieSceneScriptingBoolChannel, default=true, keys [[1001,false],[2944,true]]}]}]
   }]
 }]
display_rate {denominator=1, numerator=60}
tick_resolution {denominator=1, numerator=24000}
tracks [
 {
  type=MovieSceneCinematicShotTrack
  display_name=Shots
  sections [
   {type=MovieSceneCinematicShotSection, end_frame=2944, start_frame=1001}
   {type=MovieSceneCinematicShotSection, end_frame=3379, start_frame=2944}
   {type=MovieSceneCinematicShotSection, end_frame=3836, start_frame=3379}
   {type=MovieSceneCinematicShotSection, end_frame=4317, start_frame=3836}
   {type=MovieSceneCinematicShotSection, end_frame=4856, start_frame=4317}
   {type=MovieSceneCinematicShotSection, end_frame=5798, start_frame=4856}
   {type=MovieSceneCinematicShotSection, end_frame=10909, start_frame=8836}
   {type=MovieSceneCinematicShotSection, end_frame=10494, start_frame=9955}
   {type=MovieSceneCinematicShotSection, end_frame=11886, start_frame=11451}
   {type=MovieSceneCinematicShotSection, end_frame=12321, start_frame=11886}
   {type=MovieSceneCinematicShotSection, end_frame=20692, start_frame=12648}
  ]
 }
 {type=MovieSceneSubTrack, display_name=Sq_WildEstate_House_H, sections [{type=MovieSceneSubSection, end_frame=2944, start_frame=1001}]}
 {
  type=MovieSceneDataLayerTrack
  display_name=Data Layer
  sections [{type=MovieSceneDataLayerSection, end_frame=5798, start_frame=2944}, {type=MovieSceneDataLayerSection, end_frame=2944, start_frame=1001}]
 }
]
